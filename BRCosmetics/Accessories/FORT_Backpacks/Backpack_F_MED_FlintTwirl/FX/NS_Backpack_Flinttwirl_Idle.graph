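# Idle Niagara effect for the Flinttwirl backpack cosmetic: a blue audio-reactive glow attached to a socket.
# System-level audio reactivity, with a fake loop module, and an intro fade drive brightness, size, spawning and motion.
# GlowCenter renders pulsing sprites and a light, while StaticRings spins two glowing mesh rings with sprites.
# Curl-noise motes come in near-duplicate GPU (MotesAndGlow) and CPU (MotesAndGlowLow) emitters, with reactivity scaled down in the front end.
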NiagaraSystem /BRCosmetics/Accessories/FORT_Backpacks/Backpack_F_MED_FlintTwirl/FX/NS_Backpack_Flinttwirl_Idle
dependencies {
 usedDataInterfaces [
  &NiagaraDataInterfaceCurve
  &NiagaraDataInterfaceArrayFloat
  &NiagaraDataInterfaceArrayFloat3
  &NiagaraDataInterfaceSocketReader
  &NiagaraDataInterfaceColorCurve
 ]
 usedDynamicInputs [
  &RandomRangeInt
  /Niagara/DynamicInputs/Lerp/SmoothLerpOverTime_Float
  &RemapRange
  &MakeFloatFromLinearColor
 ]
 usedModules [
  &self:SystemSpawnScript.SystemScriptSource.SystemScriptGraph.NiagaraNodeAssignment_0.SetVariables_#1
  &self:SystemSpawnScript.SystemScriptSource.SystemScriptGraph.NiagaraNodeAssignment_1.SetVariables_#2
  &SystemState
  /Game/Effects/Niagara/Modules/Audio/FN_FakeAudioReactivityLoop
  &self:SystemSpawnScript.SystemScriptSource.SystemScriptGraph.NiagaraNodeAssignment_5.SetVariables_#3
  &Timeline
  &self:SystemSpawnScript.SystemScriptSource.SystemScriptGraph.NiagaraNodeAssignment_6.SetVariables_#4
  &self:SystemSpawnScript.SystemScriptSource.SystemScriptGraph.NiagaraNodeAssignment_7.SetVariables_#5
 ]
}
emitters {
 GlowCenter {
  emitterUpdateScript [
   &EmitterState(Life Cycle Mode=::NewEnumerator1, Loop Behavior=::NewEnumerator1, Loop Duration Mode=::NewEnumerator1, UseLoopDelay=false)
   &SpawnBurst_Instantaneous(Use Spawn Probability=false, UseLoopCountLimit=false)
  ]
  particleSpawnScript [
   &InitializeParticle(Color=<rgba 0 1.853 5 1>, Sprite Size Mode=::NewEnumerator3, Uniform Sprite Size=13, UsePositionOffset=false, Write Material Random=true)
   &CameraOffset(Camera Offset Amount=10)
   &FN_Sprite_Attributes(
    Camera Offset=15
    Color=&MakeLinearColorFromVectorAndFloat("Vector (RGB)"=&MakeVectorFromLinearColorRGB(Color=$Particles.Color), "Float (Alpha)"=0.5)
    Create Camera Offset Attribute=true
    Create Color Attribute=true
    Create SpriteSize Attribute=true
    SpriteSize=&Multiply_Vector2D_ByFloat(Float=2.25, Vector2D=$Particles.SpriteSize)
    Write to Module Namespace=true
   )
   &FN_Sprite_Attributes(
    Color=<rgba 0 0.5362 1 1>
    Create Camera Offset Attribute=true
    Create Color Attribute=true
    Create SpriteSize Attribute=true
    SpriteSize=<55 55>
    Write to Module Namespace=true
   ) {module=FN_Sprite_Attributes001}
   /SetVariables(Particles.HaloDynamic:Vector4f=<1 0 1 0>) {isSetParameters=true}
  ]
  particleUpdateScript [
   &ParticleState(Kill Particles When Lifetime Has Elapsed=false, Let Infinitely Lived Particles Die When Emitter Deactivates=true)
   &SocketLocation(
    Interpolate Socket Sampling=true
    SamplePosOffset=false
    Socket Reader=$System.Socket Reader
    Socket Sampling Mode=::NewEnumerator2
    VelocitySpeedLimit=false
   )
   /SetVariables(Particles.ReactivityMult:NiagaraFloat=&Multiply_Float(
     A=&FloatFromCurve(
      FloatCurve=DataInterface() {propertyValues {Curve {
         defaultValue=0
         keys [
          {interpMode=RCIM_Cubic, tangentMode=RCTM_User, time=0, value=0}
          {interpMode=RCIM_Cubic, tangentMode=RCTM_User, time=0.6, value=2}
          {tangentMode=RCTM_User, time=1, value=0}
         ]
        }}}
      CurveIndex=$System.AudioReactivity
     )
     B=&FN_FrontEndFloatModifier(Gameplay=1, Front End=0.75)
    )) {isSetParameters=true}
   &ScaleColor(
    Scale Alpha=$System.IntroFade
    Scale RGB=&Multiply_VectorByFloat(Vector=<1 1 1>, Float=&Lerp_Float(A=1.5, B=1, Alpha=$Particles.ReactivityMult))
    ScaleA=true
    ScaleRGB=true
   )
   &FN_Light_Attributes(
    Color=$Particles.Color
    Create Color Attribute=true
    Create Exponent Attribute=true
    Create Position Attribute=true
    Create Radius Attribute=true
    Exponent=3
    Position:NiagaraPosition=$Output.SocketLocation.SampledPosition
    Radius=30
    Scale Color=&Lerp_Float(A=0.25, B=2, Alpha=$Particles.ReactivityMult)
   )
   &InterpolateOverTime(Rate Of Change=6, Target Float Value:NiagaraFloat=$Particles.ReactivityMult)
   /SetVariables(
    Particles.FN_Sprite_Attributes.SpriteSize:Vector2f=&Multiply_Vector2D_ByFloat(
     Float=&Waveform(
      "Waveform [1]"=::NewEnumerator0
      "[1] Phase Scale"=2.5
      "[1] Frequency"=1
      "[1] Amplitude Min/Max"=<0.98 1>
      "[1] Scale"=&Lerp_Float(A=1.25, B=1.75, Alpha=$Particles.InterpolateOverTime.InterpolateOverTime_Float.Moving Average)
     )
     Vector2D=$Particles.Initial.FN_Sprite_Attributes.SpriteSize
    )
    Particles.HaloColor:LinearColor=&Multiply_LinearColorByFloat(
     Linear Color=$Particles.FN_Sprite_Attributes.Color
     Float=&Waveform(
      "Waveform [1]"=::NewEnumerator12
      "[1] Frequency"=5
      "[1] Amplitude Min/Max"=<0.85 1>
      "[1] Scale"=&Multiply_Float(
       A=&Lerp_Float(A=1, B=0.75, Alpha=$Particles.InterpolateOverTime.InterpolateOverTime_Float.Moving Average)
       B=$System.IntroFade
      )
     )
    )
   ) {isSetParameters=true}
   /SetVariables(
    Particles.FN_Sprite_Attributes001.Color:LinearColor=&Multiply_LinearColorByFloat(
     Linear Color=$Particles.Initial.FN_Sprite_Attributes001.Color
     Float=&Multiply_Float(A=&Lerp_Float(A=0, B=0.35, Alpha=$Particles.ReactivityMult), B=$System.IntroFade)
    )
    Particles.FN_Sprite_Attributes001.SpriteSize:Vector2f=&Multiply_Vector2D_ByFloat(
     Float=&Multiply_Float(
      A=&Lerp_Float(A=1.3, B=1.1, Alpha=$Particles.ReactivityMult)
      B=&MakeCustomFloatFromBool(Bool=&SetBoolByFloatComparison(A=$System.AudioReactivity, B=0), True Float=1, False Float=0)
     )
     Vector2D=$Particles.Initial.FN_Sprite_Attributes001.SpriteSize
    )
   ) {isSetParameters=true}
  ]
  rendererClasses [&NiagaraSpriteRendererProperties, &NiagaraLightRendererProperties]
  renderers [
   {rendererClass=&NiagaraSpriteRendererProperties}
   {rendererClass=&NiagaraSpriteRendererProperties, rendererIndex=1}
   {rendererClass=&NiagaraLightRendererProperties, rendererIndex=2}
   {rendererClass=&NiagaraSpriteRendererProperties, rendererIndex=3}
  ]
 }
 MotesAndGlow {
  simTarget=GPUComputeSim
  emitterUpdateScript [
   &EmitterState()
   &SpawnRate(
    Spawn Probability=&MakeCustomFloatFromBool(Bool=&SetBoolByFloatComparison(A=$System.AudioReactivity, B=0), True Float=1, False Float=0)
    SpawnRate=15
    Use Spawn Probability=true
   )
   &SpawnRate(
    Spawn Probability=&MakeCustomFloatFromBool(Bool=&SetBoolByFloatComparison(A=$System.AudioReactivity, B=0.5), True Float=1, False Float=0)
    SpawnGroup=1
    SpawnRate=15
    Use Spawn Probability=true
   ) {module=SpawnRate001}
  ]
  particleSpawnScript [
   /SetVariables(Particles.VisibilityTag=$Emitter.SpawnGroup) {isSetParameters=true, module=SetVariables_#6}
   &InitializeParticle(
    Color=&MergeHelper_ColorSwitch(0=<rgba 0.5705 2.5552 5 1>, 1=<rgba 1.7115 7.6656 15 1>, 2=?, 3=?, Index=$Emitter.SpawnGroup, Num=::NewEnumerator0)
    Lifetime=&MergeHelper_FloatSwitch(
     0=&RandomRangeFloat(Minimum=0.4, Maximum=0.8)
     1=&RandomRangeFloat(Minimum=0.4, Maximum=0.8)
     2=?
     3=?
     Index=&SpawnGroup()
     Num=::NewEnumerator0
    )
    Sprite Size Mode=::NewEnumerator3
    Uniform Sprite Size=&MergeHelper_FloatSwitch(
     0=&RandomRangeFloat(Minimum=5, Maximum=15)
     1=&RandomRangeFloat(Minimum=1.5, Maximum=2)
     2=?
     3=?
     Index=$Emitter.SpawnGroup
     Num=::NewEnumerator0
    )
    UsePositionOffset=false
    Write Material Random=true
   )
   &SocketLocation(SamplePosOffset=true, Socket Reader=$System.Socket Reader, Socket Sampling Mode=::NewEnumerator2, VelocitySpeedLimit=false)
   &ShapeLocation(
    MaskSpawnGroup=false
    Shape Origin=$Output.SocketLocation.SampledPosition
    Sphere Distribution=::NewEnumerator1
    Sphere Radius=&RandomRangeFloat(Minimum=4, Maximum=7)
    U Position=&MergeHelper_FloatSwitch(
     0=&RandomRangeFloat(Minimum=0, Maximum=1)
     1=&MergeHelper_FloatSwitch(
      0=&RandomRangeFloat(Minimum=0, Maximum=1)
      1=&RandomRangeFloat(Minimum=0, Maximum=1)
      2=?
      3=?
      Index=$Emitter.SpawnGroup
      Num=::NewEnumerator0
     )
     2=?
     3=?
     Index=$Emitter.SpawnGroup
     Num=::NewEnumerator0
    )
    UseOwnerScale=true
    V Position=&RandomRangeFloat(Minimum=0, Maximum=1)
   )
   &CameraOffset(Camera Offset Amount=&MergeHelper_FloatSwitch(
     0=&RandomRangeFloat(Minimum=15, Maximum=20)
     1=&RandomRangeFloat(Minimum=15, Maximum=20)
     2=?
     3=?
     Index=$Emitter.SpawnGroup
     Num=::NewEnumerator0
    ))
   &DynamicMaterialParameters(
    Index 0 Param 1=0.25
    Index 0 Param 2=&MergeHelper_FloatSwitch(
     0=0.3
     1=&RandomRangeFloat(Minimum=0.3, Maximum=0.5)
     2=?
     3=?
     Index=$Emitter.SpawnGroup
     Num=::NewEnumerator0
    )
    Index 0 Param 3=&RandomRangeFloat(Minimum=0.25, Maximum=0.5)
    Index 0 Param 4=1
    Module.Index0Float0Write=true
    Module.Index0Float1Write=true
    Module.Index0Float2Write=true
    Module.Index0Float3Write=true
   )
   &AddVelocity(
    Module.Use Falloff Distance=false
    Override Seed=false
    Velocity Mode=::NewEnumerator1
    Velocity Origin=$Output.SocketLocation.SampledPosition
   )
  ]
  particleUpdateScript [
   &ParticleState()
   /SetVariables(Particles.ReactivityMult:NiagaraFloat=&Multiply_Float(
     A=&FloatFromCurve(
      FloatCurve=DataInterface() {propertyValues {Curve {
         defaultValue=0
         keys [
          {interpMode=RCIM_Cubic, tangentMode=RCTM_User, time=0, value=0}
          {interpMode=RCIM_Cubic, tangentMode=RCTM_User, time=0.5, value=1}
          {tangentMode=RCTM_User, time=1, value=0}
         ]
        }}}
      CurveIndex=$System.AudioReactivity
     )
     B=&FN_FrontEndFloatModifier(Gameplay=1, Front End=0.75)
    )) {isSetParameters=true, module=SetVariables_#7}
   &ScaleSpriteSize(Uniform Curve Sprite Scale=DataInterface() {propertyValues {Curve {
       defaultValue=0
       keys [
        {arriveTangent=4.1785, interpMode=RCIM_Cubic, leaveTangent=4.1785, tangentMode=RCTM_User, time=0, value=0}
        {arriveTangent=1.1179, interpMode=RCIM_Cubic, leaveTangent=1.1179, tangentMode=RCTM_User, time=0.2, value=1}
        {arriveTangent=-1.25, interpMode=RCIM_Cubic, leaveTangent=-1.25, time=0.6596, value=0.8783}
        {arriveTangent=0.0224, interpMode=RCIM_Cubic, leaveTangent=0.0224, tangentMode=RCTM_User, time=1, value=0}
       ]
      }}})
   &ScaleColor(
    Scale Alpha=&FloatFromCurve(
     FloatCurve=DataInterface() {propertyValues {Curve {keys [
         <at 0 0 RCIM_Cubic>
         {arriveTangent=1.6667, interpMode=RCIM_Cubic, leaveTangent=1.6667, time=0.0968, value=1}
         {arriveTangent=-1.1072, interpMode=RCIM_Cubic, leaveTangent=-1.1072, time=0.6, value=1}
         <at 1 0 RCIM_Cubic>
        ]}}}
     Scale Curve=&Multiply_Float(A=&Lerp_Float(A=0, B=1, Alpha=$Particles.ReactivityMult), B=$System.IntroFade)
    )
    ScaleA=true
    ScaleRGB=true
   )
   &CurlNoiseForce(Bias Noise=false, Noise Frequency=250, Noise Strength=666, Pan Noise=false)
   &ScaleVelocity(Velocity Scale=&Multiply_VectorByFloat(Vector=<1 1 1>, Float=&Lerp_Float(A=0, B=2, Alpha=$Particles.ReactivityMult)))
   &Drag(Drag=10, Use Linear Drag=true, Use Rotational Drag=true)
   &SolveForcesAndVelocity(Clamp Velocity=false, Limit Acceleration=false)
  ]
  rendererClasses [&NiagaraSpriteRendererProperties]
  renderers [
   {rendererClass=&NiagaraSpriteRendererProperties}
   {rendererClass=&NiagaraSpriteRendererProperties, rendererIndex=1}
  ]
 }
 MotesAndGlowLow {
  emitterUpdateScript [
   &EmitterState()
   &SpawnRate(
    Spawn Probability=&MakeCustomFloatFromBool(Bool=&SetBoolByFloatComparison(A=$System.AudioReactivity, B=0), True Float=1, False Float=0)
    SpawnRate=8
    Use Spawn Probability=true
   )
   &SpawnRate(
    Spawn Probability=&MakeCustomFloatFromBool(Bool=&SetBoolByFloatComparison(A=$System.AudioReactivity, B=0.5), True Float=1, False Float=0)
    SpawnGroup=1
    SpawnRate=25
    Use Spawn Probability=true
   ) {module=SpawnRate001}
  ]
  particleSpawnScript [
   /SetVariables(Particles.VisibilityTag=$Emitter.SpawnGroup) {isSetParameters=true, module=SetVariables_#6}
   &InitializeParticle(
    Color=&MergeHelper_ColorSwitch(0=<rgba 0.5705 2.5552 5 1>, 1=<rgba 1.7115 7.6656 15 1>, 2=?, 3=?, Index=$Emitter.SpawnGroup, Num=::NewEnumerator0)
    Lifetime=&MergeHelper_FloatSwitch(
     0=&RandomRangeFloat(Minimum=0.4, Maximum=0.8)
     1=&RandomRangeFloat(Minimum=0.4, Maximum=0.8)
     2=?
     3=?
     Index=&SpawnGroup()
     Num=::NewEnumerator0
    )
    Sprite Size Mode=::NewEnumerator3
    Uniform Sprite Size=&MergeHelper_FloatSwitch(
     0=&RandomRangeFloat(Minimum=5, Maximum=15)
     1=&RandomRangeFloat(Minimum=1.5, Maximum=2)
     2=?
     3=?
     Index=$Emitter.SpawnGroup
     Num=::NewEnumerator0
    )
    UsePositionOffset=false
    Write Material Random=true
   )
   &SocketLocation(SamplePosOffset=true, Socket Reader=$System.Socket Reader, Socket Sampling Mode=::NewEnumerator2, VelocitySpeedLimit=false)
   &ShapeLocation(
    MaskSpawnGroup=false
    Shape Origin=$Output.SocketLocation.SampledPosition
    Sphere Distribution=::NewEnumerator1
    Sphere Radius=&RandomRangeFloat(Minimum=4, Maximum=7)
    U Position=&MergeHelper_FloatSwitch(
     0=&RandomRangeFloat(Minimum=0, Maximum=1)
     1=&MergeHelper_FloatSwitch(
      0=&RandomRangeFloat(Minimum=0, Maximum=1)
      1=&RandomRangeFloat(Minimum=0, Maximum=1)
      2=?
      3=?
      Index=$Emitter.SpawnGroup
      Num=::NewEnumerator0
     )
     2=?
     3=?
     Index=$Emitter.SpawnGroup
     Num=::NewEnumerator0
    )
    UseOwnerScale=true
    V Position=&RandomRangeFloat(Minimum=0, Maximum=1)
   )
   &CameraOffset(Camera Offset Amount=&MergeHelper_FloatSwitch(
     0=&RandomRangeFloat(Minimum=15, Maximum=20)
     1=&RandomRangeFloat(Minimum=15, Maximum=20)
     2=?
     3=?
     Index=$Emitter.SpawnGroup
     Num=::NewEnumerator0
    ))
   &DynamicMaterialParameters(
    Index 0 Param 1=0.25
    Index 0 Param 2=&MergeHelper_FloatSwitch(
     0=0.3
     1=&RandomRangeFloat(Minimum=0.3, Maximum=0.5)
     2=?
     3=?
     Index=$Emitter.SpawnGroup
     Num=::NewEnumerator0
    )
    Index 0 Param 3=&RandomRangeFloat(Minimum=0.25, Maximum=0.5)
    Index 0 Param 4=1
    Module.Index0Float0Write=true
    Module.Index0Float1Write=true
    Module.Index0Float2Write=true
    Module.Index0Float3Write=true
   )
   &AddVelocity(
    Module.Use Falloff Distance=false
    Override Seed=false
    Velocity Mode=::NewEnumerator1
    Velocity Origin=$Output.SocketLocation.SampledPosition
   )
  ]
  particleUpdateScript [
   &ParticleState()
   /SetVariables(Particles.ReactivityMult:NiagaraFloat=&Multiply_Float(
     A=&FloatFromCurve(
      FloatCurve=DataInterface() {propertyValues {Curve {
         defaultValue=0
         keys [
          {interpMode=RCIM_Cubic, tangentMode=RCTM_User, time=0, value=0}
          {interpMode=RCIM_Cubic, tangentMode=RCTM_User, time=0.5, value=1}
          {tangentMode=RCTM_User, time=1, value=0}
         ]
        }}}
      CurveIndex=$System.AudioReactivity
     )
     B=&FN_FrontEndFloatModifier(Gameplay=1, Front End=0.75)
    )) {isSetParameters=true, module=SetVariables_#7}
   &ScaleSpriteSize(Uniform Curve Sprite Scale=DataInterface() {propertyValues {Curve {
       defaultValue=0
       keys [
        {arriveTangent=4.1785, interpMode=RCIM_Cubic, leaveTangent=4.1785, tangentMode=RCTM_User, time=0, value=0}
        {arriveTangent=1.1179, interpMode=RCIM_Cubic, leaveTangent=1.1179, tangentMode=RCTM_User, time=0.2, value=1}
        {arriveTangent=-1.25, interpMode=RCIM_Cubic, leaveTangent=-1.25, time=0.6596, value=0.8783}
        {arriveTangent=0.0224, interpMode=RCIM_Cubic, leaveTangent=0.0224, tangentMode=RCTM_User, time=1, value=0}
       ]
      }}})
   &ScaleColor(
    Scale Alpha=&FloatFromCurve(
     FloatCurve=DataInterface() {propertyValues {Curve {keys [
         <at 0 0 RCIM_Cubic>
         {arriveTangent=1.6667, interpMode=RCIM_Cubic, leaveTangent=1.6667, time=0.0968, value=1}
         {arriveTangent=-1.1072, interpMode=RCIM_Cubic, leaveTangent=-1.1072, time=0.6, value=1}
         <at 1 0 RCIM_Cubic>
        ]}}}
     Scale Curve=&Multiply_Float(A=&Lerp_Float(A=0, B=1, Alpha=$Particles.ReactivityMult), B=$System.IntroFade)
    )
    ScaleA=true
    ScaleRGB=true
   )
   &CurlNoiseForce(Bias Noise=false, Noise Frequency=250, Noise Strength=666, Pan Noise=false)
   &ScaleVelocity(Velocity Scale=&Multiply_VectorByFloat(Vector=<1 1 1>, Float=&Lerp_Float(A=0, B=2, Alpha=$Particles.ReactivityMult)))
   &Drag(Drag=10, Use Linear Drag=true, Use Rotational Drag=true)
   &SolveForcesAndVelocity(Clamp Velocity=false, Limit Acceleration=false)
  ]
  rendererClasses [&NiagaraSpriteRendererProperties]
  renderers [
   {rendererClass=&NiagaraSpriteRendererProperties}
   {rendererClass=&NiagaraSpriteRendererProperties, rendererIndex=1}
  ]
 }
 StaticRings {
  emitterUpdateScript [
   &EmitterState(Life Cycle Mode=::NewEnumerator1, Loop Behavior=::NewEnumerator1, Loop Duration Mode=::NewEnumerator1, UseLoopDelay=false)
   &SpawnBurst_Instantaneous(Spawn Count=2, Use Spawn Probability=false, UseLoopCountLimit=false)
  ]
  particleSpawnScript [
   &PartitionParticles()
   &InitializeParticle(
    Color=<rgba 0.029 6.6465 15 1>
    Mesh Scale Mode=::NewEnumerator3
    Mesh Uniform Scale=&SelectFloatFromArray(Direct Array Index=$Particles.PartitionParticles.Partition, Float Selection Array=DataInterface() {propertyValues {FloatData [0.15,0.1325]}})
    Sprite Size Mode=::NewEnumerator3
    UsePositionOffset=false
    Write Material Random=true
   )
   &InitialMeshOrientation(
    Mesh Axis To Orient=&SelectVectorFromArray(
     Direct Array Index=$Particles.PartitionParticles.Partition
     Vector Selection Array=DataInterface() {propertyValues {FloatData [<1 0 0>, <0 0.5 1>]}}
    )
    Mesh Orientation Mode=::NewEnumerator2
    Orientation Vector=&SelectVectorFromArray(
     Direct Array Index=$Particles.PartitionParticles.Partition
     Vector Selection Array=DataInterface() {propertyValues {FloatData [<1 0 0>, <0 0 0>]}}
    )
    Rotation=&SelectVectorFromArray(
     Direct Array Index=$Particles.PartitionParticles.Partition
     Vector Selection Array=DataInterface() {propertyValues {FloatData [<0.09 0 0>, <0.5 0 0>]}}
    )
    Use Rotation Vector=true
   )
   &CameraOffset()
   &DynamicMaterialParameters(
    Index 0 Param 1=0.25
    Index 0 Param 2=&MergeHelper_FloatSwitch(
     0=0.3
     1=&RandomRangeFloat(Minimum=0.3, Maximum=0.5)
     2=?
     3=?
     Index=$Emitter.SpawnGroup
     Num=::NewEnumerator0
    )
    Index 0 Param 3=&RandomRangeFloat(Minimum=0.25, Maximum=0.5)
    Index 0 Param 4=1
    Module.Index0Float0Write=true
    Module.Index0Float1Write=true
    Module.Index0Float2Write=true
    Module.Index0Float3Write=true
   )
  ]
  particleUpdateScript [
   &SolveForcesAndVelocity(Clamp Velocity=false, Limit Acceleration=false)
   &SocketLocation(
    Interpolate Socket Sampling=true
    Orientation Sampling=::NewEnumerator1
    SamplePosOffset=false
    Socket Reader=$System.Socket Reader
    Socket Sampling Mode=::NewEnumerator2
    VelocitySpeedLimit=false
   )
   &ParticleState(Kill Particles When Lifetime Has Elapsed=false, Let Infinitely Lived Particles Die When Emitter Deactivates=true)
   /SetVariables(Particles.ReactivityMult:NiagaraFloat=&Multiply_Float(
     A=&FloatFromCurve(
      FloatCurve=DataInterface() {propertyValues {Curve {
         defaultValue=0
         keys [
          {interpMode=RCIM_Cubic, tangentMode=RCTM_User, time=0, value=0}
          {interpMode=RCIM_Cubic, tangentMode=RCTM_User, time=0.5, value=1}
          {tangentMode=RCTM_User, time=1, value=0}
         ]
        }}}
      CurveIndex=$System.AudioReactivity
     )
     B=&FN_FrontEndFloatModifier(Gameplay=1, Front End=0.75)
    )) {isSetParameters=true}
   &InterpolateOverTime(Rate Of Change=8, Target Float Value:NiagaraFloat=$Particles.ReactivityMult)
   &ScaleMeshSize(Scale Factor=&Multiply_VectorByFloat(
     Vector=<1 1 1>
     Float=&MakeCustomFloatFromBool(Bool=&SetBoolByFloatComparison(A=$System.AudioReactivity, B=0), True Float=1, False Float=0)
    ))
   &UpdateMeshOrientation(
    Rotation Rate=&Lerp_Float(A=0, B=1, Alpha=$Particles.ReactivityMult)
    Rotation Vector=&MergeHelper_VectorSwitch(0=<0.25 0.4 0.8>, 1=<0.25 -0.4 0.8>, 2=?, 3=?, Index=$Particles.PartitionParticles.Partition, Num=::NewEnumerator0)
   )
   &ScaleSpriteSize(
    Scale Sprite Size Mode=::NewEnumerator0
    Uniform Scale Factor=&MakeCustomFloatFromBool(
     Bool=&SetBoolByFloatComparison(A=$System.AudioReactivity, B=0)
     True Float=&Lerp_Float(A=0.75, B=1.25, Alpha=$Particles.InterpolateOverTime.InterpolateOverTime_Float.Moving Average)
     False Float=0
    )
   )
   /SetVariables(
    Particles.RingGlowCameraOffset=5
    Particles.RingGlowOffset:NiagaraPosition=&TransformPosition(Position=$Output.SocketLocation.SampledPosition, Source Space=::Local, Destination Space=::Simulation)
   ) {isSetParameters=true}
   &ScaleColor(
    Scale Alpha=&Multiply_Float(A=&Lerp_Float(A=0, B=1, Alpha=$Particles.ReactivityMult), B=$System.IntroFade)
    Scale RGB=&Multiply_VectorByFloat(Vector=<1 1 1>, Float=&Lerp_Float(A=0, B=5, Alpha=$Particles.ReactivityMult))
    ScaleA=true
    ScaleRGB=true
   )
   /SetVariables(Particles.SoftGlowColor:LinearColor=&Multiply_LinearColorByFloat(
     Linear Color=$Particles.Color
     Float=&Multiply_Float(A=&Lerp_Float(A=1, B=0, Alpha=$Particles.ReactivityMult), B=$System.IntroFade)
    )) {isSetParameters=true}
   /SetVariables(
    Particles.RingGlowColor:LinearColor=&Multiply_LinearColorByFloat(
     Linear Color=<rgba 0 1.6296 3 1>
     Float=&Multiply_Float(
      A=&Lerp_Float(A=0, B=3, Alpha=$Particles.InterpolateOverTime.InterpolateOverTime_Float.Moving Average)
      B=$System.IntroFade
     )
    )
    Particles.SpritePos:NiagaraPosition=&TransformPosition(
     Position=&AddVectorToPosition(
      Position=&RotateVectorByQuaternion(
       Vector To Rotate=&SelectVectorFromArray(
        Direct Array Index=$Particles.PartitionParticles.Partition
        Vector Selection Array=DataInterface() {propertyValues {FloatData [<10.5 10.5 0>, <9.25 9.25 0>]}}
       )
       Quaternion=$Output.UpdateMeshOrientation.MeshOrientation
      )
      Vector=&ConvertVectorToPosition(
       Input Position=Linked() {linkedVariable {name=Particles.RingGlowOffset, type=NiagaraPosition}}
       Conversion Operation:ENiagara_LWCConvertVecToPos=::NewEnumerator0
      )
     )
     Source Space=::Local
     Destination Space=::Simulation
    )
   ) {isSetParameters=true}
  ]
  rendererClasses [&NiagaraMeshRendererProperties, &NiagaraSpriteRendererProperties]
  renderers [
   {rendererClass=&NiagaraMeshRendererProperties, rendererIndex=0}
   {rendererClass=&NiagaraMeshRendererProperties, rendererIndex=1}
   {rendererClass=&NiagaraSpriteRendererProperties, rendererIndex=2}
  ]
 }
}
summary {
 systemName=NS_Backpack_Flinttwirl_Idle
 emitters [
  {emitterName=GlowCenter, rendererClasses [&NiagaraSpriteRendererProperties, &NiagaraLightRendererProperties]}
  {emitterName=MotesAndGlow, simTarget=GPUComputeSim, rendererClasses [&NiagaraSpriteRendererProperties]}
  {emitterName=StaticRings, rendererClasses [&NiagaraMeshRendererProperties, &NiagaraSpriteRendererProperties]}
  {emitterName=MotesAndGlowLow, rendererClasses [&NiagaraSpriteRendererProperties]}
 ]
 userVariables [
  User.bIsFrontEnd {type=NiagaraBool}
  User.bIsFrontEndPreview {type=NiagaraBool}
  User.EnableFrontendFakeAudioLoop {type=NiagaraFloat, defaultValue=0, description=""}
 ]
}
