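# Parent material for the FlintTwirl backpack cosmetic, a cloth-shaded, anime-style surface that glows in response to audio.
# It chains the shared 2023 anime shader Pre, Calculations and Post functions.
# A frontend-loop audio amplitude, sharpened by a power, scales the emissive texture and an inverted Fresnel applied twice.
# That glow adds into both lit and shaded colors, masked by the texture's blue channel and an enable parameter.
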Material /BRCosmetics/Accessories/FORT_Backpacks/Backpack_F_MED_FlintTwirl/Materials/M_Backpack_FlintTwirl_Parent
graph {
 properties {
  bAutomaticallySetUsageInEditor=false
  bUseMaterialAttributes=true
  bUsedWithClothing=true
  bUsedWithMeshDeformer=true
  bUsedWithMorphTargets=true
  bUsedWithNiagaraMeshParticles=true
  bUsedWithSkeletalMesh=true
  bUsedWithStaticLighting=true
  pixelDepthOffsetMode=PDOM_Legacy
  shadingModel=MSM_Cloth
  usedShadingModels=Cloth
 }
 nodes [
  n1 = MaterialFunctionCall!(CustomUV) {materialFunction=/Game/Global/Materials/Character_Parent_2023/Anime_Parent_2023/MaterialFunctions/MFC_AnimeShader_2023_Pre}
  n2 = MaterialFunctionCall!(CustomLightVector, OverrideNormal_WS, STT=@n1.STT, Normal_Map=@n1.STT, DFL=@n1.STT, FX=@n1.STT) {materialFunction=/Game/Global/Materials/Character_Parent_2023/Anime_Parent_2023/MaterialFunctions/MFC_AnimeShader_2023_Calculations}
  n3 = Power(
   Base=OneMinus(None=Fresnel() {baseReflectFraction=0, exponent=1})
   Exp=ScalarParameter() {defaultValue=0.5, parameterName=ReactivityFresnelExponent}
  )
  n4 = MaterialFunctionCall!(InAmplitude, InSpectrum, CustomFakeSpectrum, CustomFakeAmplitude) {materialFunction=/BRCosmetics/Global/MaterialFunctions/AudioReactivity/MF_AudioReactivity_FrontendLoop}
  n5 = ScalarParameter() {defaultValue=8, parameterName=BrightnessAtMaxAmplitude}
  n6 = TextureSample() {texture=/BRCosmetics/Accessories/FORT_Backpacks/Backpack_F_MED_FlintTwirl/Textures/T_FlintTwirl_Backpack_E}
  n7 = Multiply(A=@n3, B=@n6.RGB)
  n8 = Multiply(A=@n3, B=Multiply(A=Power(Base=@n4.Amplitude) {constExponent=6}, B=Multiply(A=@n5, B=@n7)))
  n9 = ScalarParameter() {defaultValue=1, parameterName=EnableReactivity, sliderMax=1}
  n10 = Saturate(None=Multiply(A=@n6.B) {constB=4})
  MaterialFunctionCall!(
   Shading_Gradient=@n2.Shading_Gradient
   Threshold_Gradient=@n2.Shading_Gradient
   Specular_Gradient=@n2.Shading_Gradient
   Sheen_Mask=@n2.Shading_Gradient
   Thickness=@n2.Shading_Gradient
   DFL_Mask=@n2.Shading_Gradient
   FX=@n2.Shading_Gradient
   Metallic_Gradient=@n2.Shading_Gradient
   RimLight_Gradient=@n2.Shading_Gradient
   NormalBuffer=@n2.Shading_Gradient
   --------
   CINE_LightSources=@n2.Shading_Gradient
   CINE_LightColor=@n2.Shading_Gradient
   ---------
   Color_Lit=LinearInterpolate(A=@n1.Color_Lit, B=Add(A=@n1.Color_Lit, B=@n8), Alpha=Multiply(A=@n9, B=@n10))
   Color_Shaded=LinearInterpolate(A=@n1.Color_Shaded, B=Add(A=@n1.Color_Shaded, B=@n8), Alpha=Multiply(A=@n9, B=@n10))
   SRM=@n1.SRM
   Emissive=@n1.SRM
   DFL_Color=@n1.SRM
   Specular_Color=@n1.SRM
   RimLight_Color=@n1.SRM
  ) {materialFunction=/Game/Global/Materials/Character_Parent_2023/Anime_Parent_2023/MaterialFunctions/MFC_AnimeShader_2023_Post}
 ]
}
