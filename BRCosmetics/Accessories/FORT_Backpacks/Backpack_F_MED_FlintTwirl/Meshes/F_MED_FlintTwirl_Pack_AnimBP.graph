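# Animation Blueprint for the FlintTwirl back bling, adding tape-strip physics, spine-following pack adjustment and audio-reactive ring rotation values.
# The pack bone shifts with the owner's spine yaw, then orbit and core bone modifiers build a cached pose.
# Two AnimDynamics tape chains run from it, selected by front-end state.
# Smoothed music spectrum values drive ring spin, periodically replaced by sine waves in the front end outside emotes.
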AnimBlueprint /BRCosmetics/Accessories/FORT_Backpacks/Backpack_F_MED_FlintTwirl/Meshes/F_MED_FlintTwirl_Pack_AnimBP
target_skeleton=&/F_MED_FlintTwirl_Pack_Skeleton
graphs [
 AnimGraph {nodes [
   n1 = Get GravityOverride()
   n2 = Anim Dynamics - Bone: tape_1!(
    SimulationSpace:AnimPhysSimSpaceType Enum
    ComponentPose=Local To Component(LocalPose=Use cached pose 'Main pose'())
    Alpha=1
    GravityOverride:Vector=@n1
    ExternalForce="0,0,0"
   ) -> (Pose) {
    angularDampingOverride=1
    angularSpringConstant=300
    bAngularSpring=true
    bGravityOverrideInSimSpace=true
    bOverrideAngularDamping=true
    bOverrideLinearDamping=true
    bUseGravityOverride=true
    linearDampingOverride=1
    linearSpringConstant=100
    alphaBoolBlend {blendOption=Linear}
    boundBone {boneName=tape_1}
    componentAppliedLinearAccClamp=<3000 3000 2500>
    componentLinearAccScale=<0.4 0.4 0.4>
    componentLinearVelScale=<0.4 0.4 0.4>
    physicsBodyDefinitions [{
      boundBone {boneName=tape_1}
      boxExtents=<16 16 16>
      constraintSetup {coneAngle=1, angularLimitsMax=<0 0 0>, angularLimitsMin=<0 -60 0>, angularTarget=<1 0 0>}
      localJointOffset=<2 0 0>
     }]
   }
   n3 = Anim Dynamics - Bone: tape_2!(SimulationSpace:AnimPhysSimSpaceType Enum, ComponentPose=@n2, Alpha=1, GravityOverride:Vector=@n1, ExternalForce="0,0,0") -> (Pose) {
    angularDampingOverride=0.9999
    angularSpringConstant=1000
    bAngularSpring=true
    bGravityOverrideInSimSpace=true
    bOverrideAngularDamping=true
    bOverrideLinearDamping=true
    bUseGravityOverride=true
    linearDampingOverride=1
    linearSpringConstant=100
    alphaBoolBlend {blendOption=Linear}
    boundBone {boneName=tape_2}
    componentAppliedLinearAccClamp=<3000 3000 2500>
    componentLinearAccScale=<0.3 0.3 0.3>
    componentLinearVelScale=<0.3 0.3 0.3>
    physicsBodyDefinitions [{
      boundBone {boneName=tape_2}
      boxExtents=<8 8 8>
      constraintSetup {coneAngle=1, angularLimitsMax=<0 0 0>, angularLimitsMin=<0 -75 0>, angularTarget=<1 0 0>}
      localJointOffset=<2 0 0>
     }]
   }
   n4 = Anim Dynamics - Bone: tape_3(ComponentPose=@n3, Alpha=1, ExternalForce="0,0,0") {
    angularDampingOverride=1
    angularSpringConstant=1500
    bAngularSpring=true
    bGravityOverrideInSimSpace=true
    bOverrideAngularDamping=true
    bOverrideLinearDamping=true
    bUseGravityOverride=true
    linearDampingOverride=1
    simulationSpace=RootRelative
    boundBone {boneName=tape_3}
    componentAppliedLinearAccClamp=<3000 3000 2500>
    componentLinearAccScale=<0.4 0.4 0.4>
    componentLinearVelScale=<0.4 0.4 0.4>
    physicsBodyDefinitions [{
      boundBone {boneName=tape_3}
      boxExtents=<25 25 25>
      constraintSetup {coneAngle=1, angularLimitsMax=<10 30 7>, angularLimitsMin=<-10 -75 0>, angularTarget=<1 0 0>}
      localJointOffset=<3 0 0>
     }]
   }
   n5 = Anim Dynamics - Bone: tape_3_a!(SimulationSpace:AnimPhysSimSpaceType Enum, ComponentPose=@n4, Alpha=1, GravityOverride:Vector=@n1, ExternalForce="0,0,0") -> (Pose) {
    angularDampingOverride=0.999
    angularSpringConstant=300
    bAngularSpring=true
    bOverrideAngularDamping=true
    bOverrideLinearDamping=true
    bUseGravityOverride=true
    linearDampingOverride=1
    linearSpringConstant=100
    alphaBoolBlend {blendOption=Linear}
    boundBone {boneName=tape_3_a}
    componentAppliedLinearAccClamp=<3000 3000 2500>
    componentLinearAccScale=<0.4 0.4 0.4>
    componentLinearVelScale=<0.4 0.4 0.4>
    physicsBodyDefinitions [{
      boundBone {boneName=tape_3_a}
      boxExtents=<20 20 20>
      constraintSetup {coneAngle=1, angularLimitsMax=<10 60 7>, angularLimitsMin=<-10 -65 0>, angularTarget=<1 0 0>}
      localJointOffset=<3 0 0>
     }]
   }
   n6 = Anim Dynamics - Bone: tape_4!(SimulationSpace:AnimPhysSimSpaceType Enum, ComponentPose=@n5, Alpha=1, GravityOverride:Vector=@n1, ExternalForce="0,0,0") -> (Pose) {
    angularDampingOverride=0.9999
    angularSpringConstant=300
    bAngularSpring=true
    bOverrideAngularDamping=true
    bOverrideLinearDamping=true
    bUseGravityOverride=true
    linearDampingOverride=1
    linearSpringConstant=100
    alphaBoolBlend {blendOption=Linear}
    boundBone {boneName=tape_4}
    componentAppliedLinearAccClamp=<3000 3000 2500>
    componentLinearAccScale=<0.35 0.35 0.35>
    componentLinearVelScale=<0.35 0.35 0.35>
    physicsBodyDefinitions [{
      boundBone {boneName=tape_4}
      boxExtents=<20 20 20>
      constraintSetup {coneAngle=1, angularLimitsMax=<10 50 0>, angularLimitsMin=<-10 -75 -7>, angularTarget=<1 0 0>}
      localJointOffset=<3 0 0>
     }]
   }
   n7 = Anim Dynamics - Bone: tape_5(ComponentPose=@n6, Alpha=1, ExternalForce="0,0,0") {
    angularDampingOverride=0.9999
    angularSpringConstant=1000
    bAngularSpring=true
    bGravityOverrideInSimSpace=true
    bOverrideAngularDamping=true
    bOverrideLinearDamping=true
    bUseGravityOverride=true
    linearDampingOverride=1
    simulationSpace=RootRelative
    boundBone {boneName=tape_5}
    componentAppliedLinearAccClamp=<3000 3000 2500>
    componentLinearAccScale=<0.35 0.35 0.35>
    componentLinearVelScale=<0.35 0.35 0.35>
    physicsBodyDefinitions [{
      boundBone {boneName=tape_5}
      boxExtents=<18 18 18>
      constraintSetup {coneAngle=1, angularLimitsMax=<0 0 35>, angularLimitsMin=<0 0 -10>, angularTarget=<1 0 0>}
      localJointOffset=<2 0 0>
     }]
   }
   n8 = Anim Dynamics - Bone: tape_1(ComponentPose=Local To Component(LocalPose=Use cached pose 'Main pose'()), Alpha=1, GravityOverride="0,0,-980", ExternalForce="0,0,0") {
    angularDampingOverride=1
    angularSpringConstant=300
    bAngularSpring=true
    bOverrideAngularDamping=true
    bOverrideLinearDamping=true
    bUseGravityOverride=true
    linearDampingOverride=1
    linearSpringConstant=100
    alphaBoolBlend {blendOption=Linear}
    boundBone {boneName=tape_1}
    componentAppliedLinearAccClamp=<3000 3000 2500>
    componentLinearAccScale=<0.4 0.4 0.4>
    componentLinearVelScale=<0.4 0.4 0.4>
    physicsBodyDefinitions [{
      boundBone {boneName=tape_1}
      boxExtents=<16 16 16>
      constraintSetup {coneAngle=1, angularLimitsMax=<0 0 0>, angularLimitsMin=<0 -60 0>, angularTarget=<1 0 0>}
      localJointOffset=<2 0 0>
     }]
   }
   n9 = Anim Dynamics - Bone: tape_2(ComponentPose=@n8, Alpha=1, GravityOverride="0,0,-980", ExternalForce="0,0,0") {
    angularDampingOverride=0.9999
    angularSpringConstant=1000
    bAngularSpring=true
    bOverrideAngularDamping=true
    bOverrideLinearDamping=true
    bUseGravityOverride=true
    linearDampingOverride=1
    linearSpringConstant=100
    alphaBoolBlend {blendOption=Linear}
    boundBone {boneName=tape_2}
    componentAppliedLinearAccClamp=<3000 3000 2500>
    componentLinearAccScale=<0.3 0.3 0.3>
    componentLinearVelScale=<0.3 0.3 0.3>
    physicsBodyDefinitions [{
      boundBone {boneName=tape_2}
      boxExtents=<14 14 14>
      constraintSetup {coneAngle=1, angularLimitsMax=<0 0 0>, angularLimitsMin=<0 -75 0>, angularTarget=<1 0 0>}
      localJointOffset=<2 0 0>
     }]
   }
   n10 = Get GravityOverride()
   n11 = Anim Dynamics - Bone: tape_3(ComponentPose=@n9, Alpha=1, ExternalForce=@n10) {
    angularDampingOverride=0.9999
    angularSpringConstant=1500
    bAngularSpring=true
    bGravityOverrideInSimSpace=true
    bOverrideAngularDamping=true
    bOverrideLinearDamping=true
    bUseGravityOverride=true
    linearDampingOverride=1
    simulationSpace=RootRelative
    boundBone {boneName=tape_3}
    componentAppliedLinearAccClamp=<3000 3000 2500>
    componentLinearAccScale=<0.5 0.5 0.5>
    componentLinearVelScale=<0.5 0.5 0.5>
    physicsBodyDefinitions [{
      boundBone {boneName=tape_3}
      boxExtents=<27 27 27>
      constraintSetup {coneAngle=1, angularLimitsMax=<10 30 7>, angularLimitsMin=<-10 -75 0>, angularTarget=<1 0 0>}
      localJointOffset=<3 0 0>
     }]
   }
   n12 = Anim Dynamics - Bone: tape_3_a(ComponentPose=@n11, Alpha=1, GravityOverride="0,0,-980", ExternalForce="0,0,0") {
    angularDampingOverride=0.999
    angularSpringConstant=150
    bAngularSpring=true
    bOverrideAngularDamping=true
    bOverrideLinearDamping=true
    linearDampingOverride=1
    linearSpringConstant=100
    alphaBoolBlend {blendOption=Linear}
    boundBone {boneName=tape_3_a}
    componentAppliedLinearAccClamp=<3000 3000 2500>
    componentLinearAccScale=<0.5 0.5 0.5>
    componentLinearVelScale=<0.5 0.5 0.5>
    physicsBodyDefinitions [{
      boundBone {boneName=tape_3_a}
      boxExtents=<22 22 22>
      constraintSetup {coneAngle=1, angularLimitsMax=<10 60 7>, angularLimitsMin=<-10 -75 0>, angularTarget=<1 0 0>}
      localJointOffset=<4.5 0 0>
     }]
   }
   n13 = Anim Dynamics - Bone: tape_4(ComponentPose=@n12, Alpha=1, GravityOverride="0,0,-980", ExternalForce="0,0,0") {
    angularDampingOverride=0.9999
    angularSpringConstant=300
    bAngularSpring=true
    bOverrideAngularDamping=true
    bOverrideLinearDamping=true
    linearDampingOverride=1
    linearSpringConstant=100
    alphaBoolBlend {blendOption=Linear}
    boundBone {boneName=tape_4}
    componentAppliedLinearAccClamp=<3000 3000 2500>
    componentLinearAccScale=<0.4 0.4 0.4>
    componentLinearVelScale=<0.4 0.4 0.4>
    physicsBodyDefinitions [{
      boundBone {boneName=tape_4}
      boxExtents=<27 27 27>
      constraintSetup {coneAngle=1, angularLimitsMax=<10 75 0>, angularLimitsMin=<-10 -80 -7>, angularTarget=<1 0 0>}
      localJointOffset=<3.5 0 0>
     }]
   }
   n14 = Anim Dynamics - Bone: tape_5(ComponentPose=@n13, Alpha=1, ExternalForce="0,0,0") {
    angularDampingOverride=0.9999
    angularSpringConstant=1000
    bAngularSpring=true
    bGravityOverrideInSimSpace=true
    bOverrideAngularDamping=true
    bOverrideLinearDamping=true
    bUseGravityOverride=true
    linearDampingOverride=1
    simulationSpace=RootRelative
    boundBone {boneName=tape_5}
    componentAppliedLinearAccClamp=<3000 3000 2500>
    componentLinearAccScale=<0.3 0.3 0.3>
    componentLinearVelScale=<0.3 0.3 0.3>
    physicsBodyDefinitions [{
      boundBone {boneName=tape_5}
      boxExtents=<10 10 10>
      constraintSetup {coneAngle=1, angularLimitsMax=<0 0 35>, angularLimitsMin=<0 0 -10>, angularTarget=<1 0 0>}
      localJointOffset=<3 0 0>
     }]
   }
   Output Pose(Result=Blend Poses by bool(BlendPose_0=Component To Local(ComponentPose=@n14), BlendPose_1=Component To Local(ComponentPose=@n7), bActiveValue=Get bIsInFrontEnd())) {layerGroup=None}
   n15 = "Transform (Modify) Bone - Bone: pack"(ComponentPose=Local To Component(LocalPose=Input Pose()), Translation=Get backpack_translation(), Rotation=Get backpack_rotate()) {rotationMode=BMM_Additive, translationMode=BMM_Additive, alphaBoolBlend {blendOption=Linear}, boneToModify {boneName=pack}}
   n16 = "Transform (Modify) Bone - Bone: pack"(ComponentPose=@n15, Translation="0,0,0", Rotation="0,0,-3") {rotationMode=BMM_Additive, scaleSpace=BCS_BoneSpace, translationMode=BMM_Additive, boneToModify {boneName=pack}, rotation {roll=-3}}
   n17 = "Transform (Modify) Bone - Bone: Orbit_in_2"(ComponentPose=@n16, Alpha=1, Translation="0,0,0", Rotation="0,90,-4", Scale="1,1,1") {
    rotationMode=BMM_Additive
    rotationSpace=BCS_BoneSpace
    translationMode=BMM_Additive
    alphaBoolBlend {blendOption=Linear}
    boneToModify {boneName=Orbit_in_2}
    rotation {roll=-4, yaw=90}
   }
   n18 = "Transform (Modify) Bone - Bone: Orbit_in_3"(ComponentPose=@n17, Alpha=1, Translation="0,0,0", Rotation="90,0,0", Scale="1,1,1") {
    rotationMode=BMM_Additive
    rotationSpace=BCS_BoneSpace
    translationMode=BMM_Additive
    alphaBoolBlend {blendOption=Linear}
    boneToModify {boneName=Orbit_in_3}
    rotation {pitch=90}
   }
   n19 = "Transform (Modify) Bone - Bone: Orbit_in_1"(ComponentPose=@n18, Alpha=1, Translation="0,0,0", Rotation="0,0,0", Scale="1,1,1") {
    rotationMode=BMM_Additive
    rotationSpace=BCS_BoneSpace
    translationMode=BMM_Additive
    translationSpace=BCS_BoneSpace
    alphaBoolBlend {blendOption=Linear}
    boneToModify {boneName=Orbit_in_1}
   }
   n20 = "Transform (Modify) Bone - Bone: Orbit_in_2"(ComponentPose=@n19, Alpha=1, Translation="0,0,0", Rotation="0,0,0", Scale="1,1,1") {
    rotationMode=BMM_Additive
    rotationSpace=BCS_BoneSpace
    translationMode=BMM_Additive
    translationSpace=BCS_BoneSpace
    alphaBoolBlend {blendOption=Linear}
    boneToModify {boneName=Orbit_in_2}
   }
   n21 = "Transform (Modify) Bone - Bone: Orbit_in_3"(ComponentPose=@n20, Alpha=1, Translation="0,0,0", Rotation="0,0,0", Scale="1,1,1") {
    rotationMode=BMM_Additive
    rotationSpace=BCS_BoneSpace
    translationMode=BMM_Additive
    translationSpace=BCS_BoneSpace
    alphaBoolBlend {blendOption=Linear}
    boneToModify {boneName=Orbit_in_3}
   }
   n22 = "Transform (Modify) Bone - Bone: Core"(ComponentPose=@n21, Alpha=1, Translation="0,0,0", Rotation="0,0,0", Scale="1,1,1") {
    rotationMode=BMM_Additive
    rotationSpace=BCS_BoneSpace
    translationMode=BMM_Additive
    translationSpace=BCS_BoneSpace
    alphaBoolBlend {blendOption=Linear}
    boneToModify {boneName=Core}
   }
   n23 = "Transform (Modify) Bone - Bone: Orbit_out_1"(ComponentPose=@n22, Alpha=1, Translation="0,0,0", Rotation="0,0,0", Scale="1,1,1") {
    rotationMode=BMM_Additive
    rotationSpace=BCS_BoneSpace
    translationMode=BMM_Additive
    translationSpace=BCS_BoneSpace
    alphaBoolBlend {blendOption=Linear}
    boneToModify {boneName=Orbit_out_1}
   }
   n24 = "Transform (Modify) Bone - Bone: Orbit_out_2"(ComponentPose=@n23, Alpha=1, Translation="0,0,0", Rotation="0,0,0", Scale="1,1,1") {
    rotationMode=BMM_Additive
    rotationSpace=BCS_BoneSpace
    translationMode=BMM_Additive
    translationSpace=BCS_BoneSpace
    alphaBoolBlend {blendOption=Linear}
    boneToModify {boneName=Orbit_out_2}
   }
   Save cached pose 'Main pose'(Pose=Component To Local(ComponentPose=@n24)) {cacheName=Main pose}
   Reroute Node(InputPin:Vector=@n10) -> (OutputPin:Vector)
  ]}
 EventGraph {nodes [
   n1 = Reactivity() {is=title K2Node_CustomEvent}
   > Branch(Condition=Get bIsInFrontEnd())
   > n2 = Branch(Condition=Get bIsPlayingEmote())
   Set Reactivity On/Off(execute=@n2.else, Reactivity On/Off:bool=true) -> (Output_Get:bool)
   > Delay(Duration=5)
   > Set Reactivity On/Off(Reactivity On/Off=false)
   n3 = Event BlueprintUpdateAnimation()
   > Set Delta Time X(Delta Time X=@n3.DeltaTimeX)
   > n4 = Sequence!(execute) -> (then_0, then_1, then_2:Exec, then_3:Exec, then_4:Exec)
   n5 = "Gravity Override State Function (Parameter Asset) (Nativized)"(execute=@n4.then_2, CharacterPartAnimInstance=Self-Reference(), GameTimeInSeconds=GetGameTimeInSeconds())
   > Set GravityOverride(GravityOverride=@n5)
   > Set RBAN Settings Out_Native(RBAN Settings Out_Native:Fort Rigid Body Anim Node Parameters Structure=@n5.OutRigidBodyAnimNodeParameters) -> (Output_Get:Fort Rigid Body Anim Node Parameters Structure)
   n6 = GetScalarParameterValue!(
    execute=@n4.then_0
    Collection:Material Parameter Collection Object Reference=Get MPC_Audio_Reactive_Parameters() -> (MPC_Audio_Reactive_Parameters:Material Parameter Collection Object Reference)
    ParameterName:Name=Get AmplitudeParamName() -> (AmplitudeParamName:Name=None)
   ) -> (then, ReturnValue:float) {is=title K2Node_CallMaterialParameterCollectionFunction}
   > n7 = Set Amplitude(Amplitude:float=@n6) -> (Output_Get:float)
   > Set IsEmoteAudioPlaying(IsEmoteAudioPlaying:double="To Float (Boolean)"(InBool="float > float"(A=@n7, B=0.01))) -> (Output_Get:double)
   > n8 = GetVectorParameterValue!(
    execute
    Collection:Material Parameter Collection Object Reference=Get MPC_Audio_Reactive_Parameters()
    ParameterName:Name=Get SpectrumParamName() -> (SpectrumParamName:Name=None)
   ) -> (then, ReturnValue:Linear Color Structure) {is=title K2Node_CallMaterialParameterCollectionFunction}
   n9 = Break Linear Color!(LinearColor:"Linear Color Structure (by ref)"=@n8) -> (R:float=~, G:float=~, B:float=~, A:float=~) {is=title K2Node_BreakStruct}
   n10 = Get SpectrumExp() -> (SpectrumExp:double)
   > Set Bass(Bass:float=Power!(Base:double=@n9, Exp:double=@n10) -> (ReturnValue:double) {is=title K2Node_CallFunction}) -> (Output_Get:float)
   > Set Low(Low:float=Power(Base=@n9.G, Exp=@n10)) -> (Output_Get:float)
   > Set Mid(Mid:float=Power(Base=@n9.B, Exp=@n10)) -> (Output_Get:float)
   > Set High(High:float=Power(Base=@n9.A, Exp=@n10)) -> (Output_Get:float)
   n11 = GetRealTimeSeconds()
   Set CurrentIdleRotationAmt(execute=@n4.then_1, CurrentIdleRotationAmt:double=float * float(A=float / float(A=@n11, B=-500) {is=title K2Node_CallFunction}, B=10000)) -> (Output_Get:double)
   n12 = Get CurrentIdleRotationAmt() -> (CurrentIdleRotationAmt:double)
   > Set Outer Ring Rot Z(Outer Ring Rot Z:Rotator=MakeRotator(Yaw=@n12)) -> (Output_Get:Rotator)
   > Set Outer Ring Rot X(Outer Ring Rot X:Rotator=MakeRotator(Roll=@n12)) -> (Output_Get:Rotator)
   > n13 = Sequence!(execute) -> (then_0, then_1, then_2)
   n14 = Get Bass() -> (Bass:float)
   n15 = Get FadedBass() -> (FadedBass:float)
   n16 = "float < float"(A=@n15, B=@n14)
   n17 = Get ReactiveRampUpSpeed() -> (ReactiveRampUpSpeed:float)
   n18 = Get ReactiveRampDownSpeed() -> (ReactiveRampDownSpeed:float)
   > Set FadedBass(FadedBass:float=FInterpTo(Current=@n15, Target=@n14, DeltaTime=Get Delta Time X(), InterpSpeed=SelectFloat(A=@n17, B=@n18, bPickA=@n16))) -> (Output_Get:float)
   n19 = Get Mid() -> (Mid:float)
   n20 = Get FadedMid() -> (FadedMid:float)
   n21 = "float < float"(A=@n20, B=@n19)
   n22 = Get ReactiveRampUpSpeed()
   n23 = Get ReactiveRampDownSpeed()
   > Set FadedMid(FadedMid:float=FInterpTo(Current=@n20, Target=@n19, DeltaTime=Get Delta Time X(), InterpSpeed=SelectFloat(A=@n22, B=@n23, bPickA=@n21))) -> (Output_Get:float)
   n24 = Get High() -> (High:float)
   n25 = Get FadedHigh() -> (FadedHigh:float)
   n26 = "float < float"(A=@n25, B=@n24)
   n27 = Get ReactiveRampUpSpeed()
   n28 = Get ReactiveRampDownSpeed()
   > Set FadedHigh(FadedHigh:float=FInterpTo(Current=@n25, Target=@n24, DeltaTime=Get Delta Time X(), InterpSpeed=SelectFloat(A=@n27, B=@n28, bPickA=@n26))) -> (Output_Get:float)
   n29 = Get Amplitude() -> (Amplitude:float)
   n30 = Get FadedAmplitude() -> (FadedAmplitude:float)
   n31 = "float < float"(A=@n30, B=@n29)
   n32 = Get ReactiveRampUpSpeed()
   n33 = Get ReactiveRampDownSpeed()
   > Set FadedAmplitude(FadedAmplitude:float=FInterpTo(Current=@n30, Target=@n29, DeltaTime=Get Delta Time X(), InterpSpeed=SelectFloat(A=@n32, B=@n33, bPickA=@n31))) -> (Output_Get:float)
   Branch(execute=@n13.then_1, Condition=Get Reactivity On/Off() -> (Reactivity On/Off:bool))
   n34 = GetRealTimeSeconds()
   n35 = float * float(A=1.2, B=@n34)
   > Set FadedBass(FadedBass=float * float(A="Sin (Radians)"(A=@n35), B=1.25))
   n36 = float * float(A=1.3, B=@n34)
   > Set FadedMid(FadedMid=float * float(A="Sin (Radians)"(A=@n36), B=1.25))
   n37 = float * float(A=1.4, B=@n34)
   > Set FadedAmplitude(FadedAmplitude=float * float(A="Sin (Radians)"(A=@n37), B=1.25))
   n38 = float * float(A=1, B=@n34)
   > Set FadedHigh(FadedHigh=float * float(A="Sin (Radians)"(A=@n38), B=1.25))
   n39 = float + float(
    A=float * float!(A=Get FadedBass(), B=Get RingReactiveSpinAmt() -> (RingReactiveSpinAmt:double), C:double=2) -> (ReturnValue) {is=title K2Node_PromotableOperator}
    B=Get CurrentIdleRotationAmt()
   )
   Set Ambient Rotation Ring1(execute=@n13.then_2, Ambient Rotation Ring1:Rotator=MakeRotator(Pitch=@n39, Yaw=@n39)) -> (Output_Get:Rotator)
   n40 = float + float(A=float * float(A=Get FadedMid(), B=Get RingReactiveSpinAmt(), C=1), B=Get CurrentIdleRotationAmt())
   > Set Ambient Rotation Ring2(Ambient Rotation Ring2:Rotator=MakeRotator(Roll=@n40, Yaw=@n40)) -> (Output_Get:Rotator)
   n41 = float + float(A=float * float(A=Get FadedAmplitude(), B=Get RingReactiveSpinAmt(), C=1), B=Get CurrentIdleRotationAmt())
   > Set Ambient Rotation Ring3(Ambient Rotation Ring3:Rotator=MakeRotator(Pitch=@n41, Yaw=@n41)) -> (Output_Get:Rotator)
   n42 = float * float(A=Get FadedHigh(), B=Get RingReactiveSpinAmt(), C=1)
   n43 = Get CurrentIdleRotationAmt()
   > Set Ambient Rotation Y_2(Ambient Rotation Y_2:Rotator=MakeRotator(Pitch=float + float(A=@n42, B=@n43))) -> (Output_Get:Rotator)
   n44 = GetOwnerSkeletalMeshComponent()
   n45 = BreakRotator(InRot=BreakTransform(InTransform=GetDeltaTransformFromRefPose(self=@n44, BoneName=spine_05, BaseName=spine_01)).Rotation)
   n46 = MapRangeClamped(Value=@n45.Yaw, InRangeA=0, InRangeB=-150, OutRangeB=70)
   Set backpack_rotate(execute=@n4.then_3, backpack_rotate=MakeRotator(Roll=@n46, Pitch=float / float(A=@n46, B=5) {is=title K2Node_PromotableOperator}))
   > Set backpack_translation(backpack_translation=MakeVector(Y=MapRangeClamped(Value=@n45.Yaw, InRangeA=0, InRangeB=-70, OutRangeB=-9)))
   > n47 = Branch(Condition=Get bIsInFrontEnd())
   > Set SimlationSpace(SimlationSpace:AnimPhysSimSpaceType Enum=Component) -> (Output_Get:AnimPhysSimSpaceType Enum)
   Set SimlationSpace(execute=@n47.else)
   Event BlueprintBeginPlay() {is=title K2Node_Event}
   > Parent: BlueprintBeginPlay() {is=title K2Node_CallParentFunction}
   > Delay(Duration=0)
   > Reactivity() {is=title K2Node_CallFunction}
   > Set Timer by Event(Delegate=@n1, Time=13, bLooping=true)
   TryGetPawnOwner()
   TryGetPawnOwner()
   n48 = Get Low() -> (Low:float)
   n49 = Get FadedLow() -> (FadedLow:float)
   n50 = "float < float"(A=@n49, B=@n48)
   n51 = Get ReactiveRampUpSpeed()
   n52 = Get ReactiveRampDownSpeed()
   Set FadedLow(FadedLow:float=FInterpTo(Current=@n49, Target=@n48, DeltaTime=Get Delta Time X(), InterpSpeed=SelectFloat(A=@n51, B=@n52, bPickA=@n50))) -> (Output_Get:float)
  ]}
]
variables [
 Ambient Rotation Ring1 {type=Rotator}
 Ambient Rotation Ring2 {type=Rotator}
 Ambient Rotation Ring3 {type=Rotator}
 Ambient Rotation Y_2 {type=Rotator}
 Amplitude {type=number}
 AmplitudeParamName {type=string, default=Amplitude_Music}
 backpack_rotate {type=Rotator}
 backpack_translation {type=Vector}
 Bass {type=number}
 CurrentIdleRotationAmt {type=number}
 Delta Time X {type=number}
 FadedAmplitude {type=number}
 FadedBass {type=number}
 FadedHigh {type=number}
 FadedLow {type=number}
 FadedMid {type=number}
 GravityOverride {type=Vector}
 High {type=number}
 IsEmoteAudioPlaying {type=number}
 Low {type=number}
 Mid {type=number}
 MPC_Audio_Reactive_Parameters {type=MaterialParameterCollection, default=/Game/Athena/Events/AudioAnalysis/MPC_Audio_Reactive_Parameters}
 Outer Ring Rot X {type=Rotator}
 Outer Ring Rot Z {type=Rotator}
 Range {type=number, default=1}
 RBAN Settings Out_Native {type=FortRigidBodyAnimNodeParameters}
 ReactiveRampDownSpeed {type=number, default=5}
 ReactiveRampUpSpeed {type=number, default=10}
 Reactivity On/Off {type=boolean}
 RingReactiveSpinAmt {type=number, default=-150}
 SimlationSpace {type=AnimPhysSimSpaceType}
 SpectrumExp {type=number, default=1}
 SpectrumParamName {type=string, default=Spectrum_Music}
]
